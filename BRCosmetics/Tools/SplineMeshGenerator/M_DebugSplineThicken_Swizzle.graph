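# Translucent unlit debug material for the SplineMeshGenerator tool, set up around SplineThicken and enabled for Niagara mesh particles.
# Blue vertex color times a scalar parameter feeds both base and tip width of SplineThicken.
# A separate chain multiplies UV red by an OceanBreeze spline texture panned along U.
# The file shows no connections to material outputs; the color parameter, normal and tangent nodes sit unconnected.
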Material /BRCosmetics/Tools/SplineMeshGenerator/M_DebugSplineThicken_Swizzle
graph {
 properties {
  bUsedWithNiagaraMeshParticles=true
  blendMode=BLEND_Translucent
  pixelDepthOffsetMode=PDOM_Legacy
  shadingModel=MSM_Unlit
  usedShadingModels=Unlit
 }
 nodes [
  n1 = Multiply(A=VertexColor().B, B=ScalarParameter() {defaultValue=10})
  MaterialFunctionCall(WidthBase=@n1, WidthTip=@n1, Expand U or V UV Channel=StaticBool()) {materialFunction=&SplineThicken}
  n2 = TextureCoordinate()
  n3 = Time() {bOverride_Period=true, period=114514}
  Multiply(
   A=MaterialFunctionCall(Float2=TextureCoordinate()) {materialFunction=&BreakOutFloat2Components}.R
   B=TextureSample(UVs=Panner(Coordinate=@n2, Time=@n3) {speedX=0.25}) {texture=/Game/Accessories/FORT_Backpacks/Backpack_M_MED_OceanBreeze/FX/T_OceanBreeze_SplineA}.RGB
  )
  VectorParameter() {defaultValue=<rgba 5 4.434 3.2995 1>}
  VertexNormalWS()
  VertexTangentWS()
 ]
}
